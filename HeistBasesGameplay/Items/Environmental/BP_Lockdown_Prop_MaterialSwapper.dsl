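; Lockdown prop child Blueprint that swaps a material on its static mesh to reflect alarm state.
; UpdateAlarmVisuals gets the active state from its parent implementation and returns it unchanged.
; When active, it caches the material at the configured index and applies the override material; when inactive, it reapplies the cached original.
; BeginPlay and construction only call the parent, while overlap and tick are empty.
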(event EventBeginPlay
  (|CallParentFunction))

(event Collision|EventActorBeginOverlap
  (params (OtherActor :type Actor%20Object%20Reference)))

(event EventTick
  (params (DeltaSeconds :type Float%20%28single-precision%29)))

(fn UpdateAlarmVisuals
  (params (ForceOff :type Boolean) (NewState :type Boolean))
  (bind _materialindextoreplace (Variables|Default|GetMaterialIndexToReplace))
  (bind _staticmeshcomponent (Variables|Debug|GetStaticMeshComponent))
  (bind _returnvalue (Rendering|Material|GetMaterial _staticmeshcomponent _materialindextoreplace))
  (bind _isactive (|CallParentFunction ForceOff NewState))
  (Utilities|FlowControl|Sequence
    (:then_0
      (Variables|Debug|GetStaticMeshComponent
        (:then
          (if _isactive
            (Utilities|IsValid _returnvalue
              (:Is%20Valid
                (Variables|Default|SetOriginalMaterialCached _returnvalue)
                (bind _overridematerial (Variables|Default|GetOverrideMaterial)
                  (:then
                    (Rendering|Material|SetMaterial (Variables|Debug|GetStaticMeshComponent) (Variables|Default|GetMaterialIndexToReplace) _overridematerial))
                  (:else
                    _))))
            (else
              (bind _originalmaterialcached (Variables|Default|GetOriginalMaterialCached)
                (:then
                  (Rendering|Material|SetMaterial (Variables|Debug|GetStaticMeshComponent) (Variables|Default|GetMaterialIndexToReplace) _originalmaterialcached))
                (:else
                  _))))
          (Variables|Default|SetOriginalMaterialCached 0))
        (:else)))
    (:then_1
      (return _isactive))))

(fn ConstructionScript
  (|CallParentFunction))
