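; Parent Blueprint for environmental props that react to heist lockdowns by switching an alarm on and off.
; Lockdown initiation sets a replicated AlarmState, while expiry or reset clears it, each flushing net dormancy.
; OnRep applies changes on cosmetic-capable instances after a random delay of up to two seconds, staggering props.
; UpdateAlarmVisuals activates or fades out audio and returns the active state; construction and BeginPlay force it off.
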(event EventOnLockdownExpired
  (params (DisableVisuals :type Boolean) (DisableSounds :type Boolean))
  ;[1 1 1] Lockdown turned off
  (Variables|Default|SetAlarmState false)
  (Networking|FlushNetDormancy))

(event EventBeginPlay
  ;[1 1 1] Force all props to be off on first load
  (|CallParentFunction)
  (Networking|CanExecuteCosmeticEvents
    (:True
      (Utilities|FlowControl|DoOnce false
        (:Completed
          (CallFunction|UpdateAlarmVisuals true))))))

(event EventOnLockdownReset
  ;[1 1 1] Lockdown turned off
  (Variables|Default|SetAlarmState false)
  (Networking|FlushNetDormancy))

(event Custom|AlarmStateChanged
  (params (NewState :type Boolean))
  (Utilities|FlowControl|Delay (Math|Random|RandomFloatinRange 0.0 2.0))
  (CallFunction|UpdateAlarmVisuals false NewState))

(fn OnLockdownInitiated
  (Variables|Default|SetAlarmState true)
  (Networking|FlushNetDormancy)
  (return true))

(fn OnRep_AlarmState
  (Networking|CanExecuteCosmeticEvents
    (:True
      (CallFunction|AlarmStateChanged (Variables|Default|GetAlarmState)))
    (:False
      _))
  (return))

(fn UpdateAlarmVisuals
  (params (ForceOff :type Boolean) (NewState :type Boolean))
  (if ForceOff
    (LocalVariables|SetLocalActive false)
    (else
      (LocalVariables|SetLocalActive NewState)))
  (if (LocalVariables|GetLocalActive)
    ;[1 1 1] Audio
    (bind _audio (Variables|Default|GetAudio)
      (:then
        (Components|Activation|SetActive _audio true false))
      (:else))
    (else
      (bind _audio_2 (Variables|Default|GetAudio)
        (:then
          (Audio|Components|Audio|FadeOut _audio_2 1.0 0.0 "Linear" :!class /Script/Engine.AudioComponent))
        (:else))))
  (return (LocalVariables|GetLocalActive)))

(fn ConstructionScript
  (Networking|CanExecuteCosmeticEvents
    (:True
      (CallFunction|UpdateAlarmVisuals true))))
